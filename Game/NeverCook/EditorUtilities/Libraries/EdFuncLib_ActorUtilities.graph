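# Editor utility function library that finds the Asset Registry data of the asset behind an actor's class.
# GetAssetPackageNameForActor splits the class's soft path at the last dot to get the package name, returning None for native /Script/ classes.
# GetAssetDataForActor queries the registry with that name, including assets not on disk, and returns the first match with a validity flag.
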EditorUtilityBlueprint /Game/NeverCook/EditorUtilities/Libraries/EdFuncLib_ActorUtilities
graphs [
 GetAssetDataForActor {nodes [
   n1 = GetAssetDataForActor!() -> (then, Actor:Actor Object Reference) {is=title K2Node_FunctionEntry}
   > n2 = GetAssetPackageNameForActor!(execute, Actor:Actor Object Reference=@n1, __WorldContext:Object Reference=@n1) -> (then, PackageName:Name=None)
   > n3 = GetAssetsByPackageName!(
    execute
    self:Asset Registry Interface=GetAssetRegistry()
    PackageName:Name=@n2
    bIncludeOnlyOnDiskAssets:bool=false
    bSkipARFilteredAssets:bool=true
   ) -> (then, OutAssetData:Array of Asset Data Structures, ReturnValue:bool)
   > n4 = Branch(Condition=@n3.ReturnValue)
   > Return Node!(
    execute
    ValidAsset:bool=true
    AssetData:Asset Data Structure="Get (a copy)"(Array:Array of Asset Data Structures=@n3) -> (Output:Asset Data Structure)
   )
   Return Node(execute=@n4.else, ValidAsset=false)
  ]}
 GetAssetPackageNameForActor {nodes [
   n1 = GetAssetPackageNameForActor!() -> (then, Actor:Actor Object Reference) {is=title K2Node_FunctionEntry}
   > Is Valid(InputObject=@n1)
   n2 = Get Class(Object=@n1)
   n3 = Split(
    SourceString=BreakSoftClassPath!(InSoftClassPath:Soft Class Path Structure=GetSoftClassPath!(Class:Object Class Reference=@n2) -> (ReturnValue:Soft Class Path Structure)) -> (PathString:String)
    InStr=.
    SearchDir=FromEnd
   )
   > n4 = Branch(Condition=@n3.ReturnValue)
   > n5 = Branch(Condition=Contains(SearchIn=@n3, Substring=/Script/))
   Return Node!(execute=[@n5.then, @n4.else], PackageName:Name=None)
   Return Node(execute=@n5.else, PackageName=String To Name(InString=@n3))
  ]}
]
variables []
